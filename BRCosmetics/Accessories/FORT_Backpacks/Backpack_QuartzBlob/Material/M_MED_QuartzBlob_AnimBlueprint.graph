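# Animation Blueprint for the QuartzBlob backpack on the Celestial backpack skeleton, copying pose and curves from the attached parent mesh.
# Each update computes pack_float, a floating-motion offset built from scaled sine waves of real time, with Y using a negated absolute sine.
# A Transform (Modify) Bone chain applies that offset additively to Bone001 but is not connected to Output Pose, which receives only the copied pose.
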AnimBlueprint /BRCosmetics/Accessories/FORT_Backpacks/Backpack_QuartzBlob/Material/M_MED_QuartzBlob_AnimBlueprint
target_skeleton=/Game/Accessories/FORT_Backpacks/M_MED_Celestial_Backpack/M_MED_Celestial_Skeleton
graphs [
 AnimGraph {nodes [
   n1 = "Transform (Modify) Bone - Bone: Bone001"(ComponentPose=Local To Component(), Alpha=1, Translation=Get pack_float() -> (pack_float:Vector), Rotation="0,0,0", Scale="1,1,1") {
    translationMode=BMM_Additive
    translationSpace=BCS_BoneSpace
    alphaBoolBlend {blendOption=Linear}
    boneToModify {boneName=Bone001}
    translation {y=10}
   }
   Component To Local(ComponentPose=@n1)
   Output Pose(Result=Copy Pose From Mesh() {bCopyCurves=true, bUseAttachedParent=true, rootBoneToCopy=Backpack_BR}) {layerGroup=None}
  ]}
 EventGraph {nodes [
   Event BlueprintUpdateAnimation()
   > Sequence!(execute) -> (then_0, then_1, then_2:Exec, then_3:Exec, then_4:Exec, then_5:Exec, then_6:Exec)
   n1 = GetRealTimeSeconds()
   n2 = float * float(A=float * float(A=1, B=1), B=@n1)
   n3 = float * float(A="Absolute (Float)"(A="Sin (Radians)"(A=@n2)), B=-0.3)
   n4 = float * float(A=1, B=1)
   n5 = float * float(A="Sin (Radians)"(A=float * float(A=@n4, B=@n1)), B=1)
   n6 = float * float(A=1, B=1.5)
   n7 = float * float(A="Sin (Radians)"(A=float * float(A=@n6, B=@n1)), B=0.9)
   > Set pack_float(pack_float:Vector=vector * vector(A=MakeVector(X=@n7, Y=@n3, Z=@n5), B:Vector=MakeVector(X=15, Y=15, Z=5)) -> (ReturnValue=~)) -> (Output_Get:Vector)
   TryGetPawnOwner()
   float * float(A=1, B=15)
  ]}
]
variables [pack_float {type=Vector}]
