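# Metallic material for the QuartzBlob back bling that shades a silver color ramp by view angle, usable on Niagara mesh particles.
# A Fresnel term on a rerouted normal, scaled up, indexes a parameterized curve atlas row to produce base color.
# Constants make it fully metallic with zero specular and roughness, for a mirror-like Default Lit surface.
# Material attributes gather these, along with Normal and Opacity Mask inputs from named reroutes.
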Material /BRCosmetics/Accessories/FORT_Backpacks/Backpack_QuartzBlob/Material/M_QuartzBlob_Metalish
graph {
 properties {
  bUseMaterialAttributes=true
  bUsedWithNiagaraMeshParticles=true
  pixelDepthOffsetMode=PDOM_Legacy
  shadingModel=MSM_FromMaterialExpression
  usedShadingModels=Default Lit
 }
 nodes [
  n1 = NamedRerouteUsage()
  n2 = CurveAtlasRowParameter(CurveTime=Multiply(A=Fresnel(Normal=@n1) {baseReflectFraction=0, exponent=2}) {constB=10}) {
   atlas=&AthenaCurveAtlas01
   curve=/Game/Characters/Player/Female/Medium/Bodies/F_MED_Network/Textures/Curve_Network_Silver
   defaultValue=314
   group=QuartzBlob
   parameterName=QB ColorCurve
  }
  n3 = Constant()
  SetMaterialAttributes!(
   MaterialAttributes
   Base Color=NamedRerouteDeclaration(None=@n2) {name=baseColor, nodeColor=<rgba 0.8516 0 1 1>}
   Metallic=NamedRerouteDeclaration(None=Constant() {r=1}) {name=Metallic, nodeColor=<rgba 0.5469 1 0 1>}
   Specular=NamedRerouteDeclaration(None=@n3) {name=Specular, nodeColor=<rgba 0 1 0.1328 1>}
   Roughness=NamedRerouteDeclaration(None=@n3) {name=Roughness, nodeColor=<rgba 1 0 0.375 1>}
   Opacity Mask=NamedRerouteUsage()
   Normal=NamedRerouteUsage()
  ) {attributeSetTypes [
    #1
    #2
    #3
    #4
    #5
    #6
   ]}
 ]
}
