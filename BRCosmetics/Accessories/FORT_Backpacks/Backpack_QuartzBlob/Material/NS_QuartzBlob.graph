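# Niagara system for the QuartzBlob backpack cosmetic that drives four mesh-rendered spheres forming a wobbling blob.
# SphereSources springs particles toward jittered User.Skelmesh samples with curl noise, then rescales, clamps and fixes stuck ones.
# It reads its own particles through attribute readers to gather positions and min/max bounds, and takes color from Blueprint-fed game time.
# Disabled Orbiting_Bits would spring sprite and mesh bits around those spheres.
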NiagaraSystem /BRCosmetics/Accessories/FORT_Backpacks/Backpack_QuartzBlob/Material/NS_QuartzBlob
dependencies {
 usedDataInterfaces [&NiagaraDataInterfaceParticleRead, &NiagaraDataInterfaceCurve]
 usedDynamicInputs []
 usedModules [&SystemState]
}
emitters {
 Orbiting_Bits {
  bEnabled=false
  emitterUpdateScript [
   &EmitterState(Life Cycle Mode=::NewEnumerator1, Loop Behavior=::NewEnumerator1, Loop Duration Mode=::NewEnumerator1, UseLoopDelay=false)
   &SpawnBurst_Instantaneous(Spawn Count=15, Use Spawn Probability=false, UseLoopCountLimit=false)
  ]
  particleSpawnScript [
   &InitializeParticle(
    Color Mode=::NewEnumerator0
    Lifetime=5
    Mesh Scale Mode=::NewEnumerator3
    Mesh Uniform Scale=&RandomRangeFloat(Minimum=0.3, Maximum=2)
    UsePositionOffset=false
    Write Material Random=true
   )
   /SetVariables(Particles.NucleusParticle_idx:NiagaraInt32=&ModuloInt(A=&ReturnParticleIDAttributes(), B=4)) {isSetParameters=true}
   /SetVariables(Particles.InitialScale:Vector3f=$Particles.Scale) {isSetParameters=true}
   &ShapeLocation(
    Apply To Particle Position=::NewEnumerator1
    MaskSpawnGroup=false
    Override Seed=false
    Shape Origin=&AddVectorToPosition(Position=$Engine.Owner.Position, Vector=<0 0 0>)
    Sphere Radius=&Multiply_Float(A=1, B=1.3)
    UseOwnerScale=true
   ) {enabled=false, module=ShapeLocation001}
  ]
  particleUpdateScript [
   &ParticleState(Kill Particles When Lifetime Has Elapsed=false)
   &CurlNoiseForce(
    Noise Frequency=0.1
    Noise Quality / Cost=::NewEnumerator0
    Noise Strength=111
    Pan Noise=true
    Pan Noise Field=&RandomVector(Vector Scale=0.3)
   )
   &self:PAR_NucleusPosition_01(PAR=DataInterface() {propertyValues {EmitterBinding {bindingMode=Other, emitterName=SphereSources}}}) {module=PAR_NucleusPosition_01}
   &ShapeLocation(
    Apply To Particle Position=::NewEnumerator1
    MaskSpawnGroup=false
    Override Seed=false
    Shape Origin=&AddVectorToPosition(Position=$Particles.TargetPosition, Vector=<0 0 0>)
    Sphere Radius=&Multiply_Float(A=$Particles.TargetRadius, B=1.3)
    UseOwnerScale=true
   )
   /Niagara/Modules/Update/Forces/SpringForce(
    " Particle Equilibrium Position":Vector3f=$Particles.ShapeLocation.ShapePosition
    Attachment Point Velocity:Vector3f=$Particles.TargetVelocity
    Dampening Coefficient:NiagaraFloat=&RandomRangeFloat(Minimum=4, Maximum=1)
    Desired Minimum Separation Distance:NiagaraFloat=&Multiply_Float(A=$Particles.TargetRadius, B=1)
    Spring Tightness=15
   ) {enabled=false, module=SpringForce001}
   &SpringForce(
    " Particle Equilibrium Position":NiagaraPosition=$Particles.TargetPosition
    Attachment Point Velocity:Vector3f=$Particles.TargetVelocity
    Dampening Coefficient:NiagaraFloat=&RandomRangeFloat(Minimum=4, Maximum=1)
    Desired Minimum Separation Distance:NiagaraFloat=&Multiply_Float(A=$Particles.TargetRadius, B=1)
    Spring Tightness=15
   )
   &SolveForcesAndVelocity(Clamp Velocity=false, Limit Acceleration=false)
   &self:DistFrom_Sphere_Center_01(
    Input Max=1
    Input Min:NiagaraFloat=&RandomRangeFloat(Minimum=25, Maximum=75)
    New Curve for Floats=DataInterface() {propertyValues {Curve {keys [
        {arriveTangent=3.1804, interpMode=RCIM_Cubic, leaveTangent=3.1804, tangentMode=RCTM_User, time=0, value=0.3}
        {arriveTangent=1.1667, interpMode=RCIM_Cubic, leaveTangent=1.1667, time=0.3, value=0.75}
        {arriveTangent=-0.3573, interpMode=RCIM_Cubic, leaveTangent=-0.3573, tangentMode=RCTM_User, time=0.6, value=1}
        {arriveTangent=-3.0373, interpMode=RCIM_Cubic, leaveTangent=-3.0373, tangentMode=RCTM_User, time=1, value=0.1}
       ]}}}
    PAR=DataInterface() {propertyValues {EmitterBinding {bindingMode=Other, emitterName=SphereSources}}}
   ) {module=DistFrom_Sphere_Center_01}
   &CameraOffset(Camera Offset Amount=5)
  ]
  rendererClasses [&NiagaraSpriteRendererProperties, &NiagaraMeshRendererProperties]
  renderers [
   {rendererClass=&NiagaraSpriteRendererProperties}
   {rendererClass=&NiagaraMeshRendererProperties, rendererIndex=1}
  ]
 }
 SphereSources {
  emitterUpdateScript [
   &EmitterState(
    Inactive Response=::NewEnumerator1
    Life Cycle Mode=::NewEnumerator1
    Loop Behavior=::NewEnumerator1
    Loop Duration Mode=::NewEnumerator1
    Scalability Mode=::NewEnumerator1
    UseLoopDelay=false
   )
   &SpawnBurst_Instantaneous(Spawn Count=4, Use Spawn Probability=false, UseLoopCountLimit=false)
  ]
  particleSpawnScript [
   &InitializeParticle(
    Color Maximum=<rgba 0.5483 1 0.2361 1>
    Color Minimum=<rgba 0.1797 0.0264 0.3698 1>
    Color Mode=::NewEnumerator2
    Lifetime=5
    Mass Max=0.65
    Mass Min=0.6
    Mass Mode=::NewEnumerator3
    Mesh Scale Mode=::NewEnumerator3
    Position Offset=<0 10 -50>
    Sprite Size Mode=::NewEnumerator3
    UsePositionOffset=true
    Write Material Random=true
   )
   &SkeletalMeshLocation(
    Bone Sampling Mode=::NewEnumerator4
    MaskSpawnGroup=false
    Orientation Sampling=::NewEnumerator1
    Position Sampling=::NewEnumerator2
    Skeletal Mesh=$User.Skelmesh
   ) {enabled=false, module=SkeletalMeshLocation001}
   /SetVariables(Particles.Position=&AddVectorToPosition(Position=$Engine.Owner.Position, Vector=<0 -15 0>)) {enabled=false, isSetParameters=true}
  ]
  particleUpdateScript [
   /SetVariables(Particles.Offset:Vector3f=&RandomRangeVector(Minimum=<-8 -8 -8>, Maximum=<8 8 8>)) {isSetParameters=true}
   &SkeletalMeshLocation(
    Bone Sampling Mode=::NewEnumerator4
    MaskSpawnGroup=false
    Orientation Sampling=::NewEnumerator1
    Position Sampling=::NewEnumerator2
    Sampled Position Offset=$Particles.Offset
    Skeletal Mesh=$User.Skelmesh
   )
   &ParticleState(Kill Particles When Lifetime Has Elapsed=false)
   &CurlNoiseForce(
    Noise Frequency=0.15
    Noise Quality / Cost=::NewEnumerator0
    Noise Strength=90
    Pan Noise=true
    Pan Noise Field=&RandomVector(Vector Scale=0.5)
   )
   &Drag(Drag=5, Use Linear Drag=true, Use Rotational Drag=false) {enabled=false}
   &Color(Color=/Niagara/DynamicInputs/TypeConversions/MakeLinearColorFromFloat(Value:NiagaraFloat=$User.GameTimeFromBP))
   &SpringForce(
    " Particle Equilibrium Position":NiagaraPosition=&Add_Vector(
     A=&AddVectorToPosition(Position=$Particles.SkeletalMeshLocation.SampledPosition, Vector=<0 0 0>)
     B=&RandomRangeVector(Minimum=<-8 -8 -8>, Maximum=<8 8 8>)
    )
    Attachment Point Velocity:Vector3f=/Niagara/DynamicInputs/Math/Clamp_Vector(
     Vector:Vector3f=&Multiply_VectorByFloat(Vector=$Particles.SkeletalMeshLocation.SampledVelocity, Float=0)
     Min:Vector3f=<-200 -200 -200>
     Max:Vector3f=<200 200 200>
    )
    Dampening Coefficient=8
    Desired Minimum Separation Distance=4
    Force Strength:NiagaraFloat=&RandomRangeFloat(Minimum=0.75, Maximum=2)
    Spring Tightness=54
   ) {module=SpringForce001}
   &SolveForcesAndVelocity(Clamp Velocity=false, Limit Acceleration=false, Manually Enable Rotational Solver=true)
   /SetVariables(Particles.OldVelocity:Vector3f=$Particles.Velocity) {isSetParameters=true}
   &self:ScaleSpheres_01(Input Max=18, Input Min=8, Output Max=0.5, Output Min=5) {module=ScaleSpheres_01}
   /SetVariables(Particles.Radius:NiagaraFloat=&MakeFloatFromVector(Vector=$Particles.Scale, Channel=::NewEnumerator0)) {isSetParameters=true}
   &self:ScratchModule(Offset=-15, Weight=0.3) {module=ScratchModule}
   &self:PAR_GatherPositions_02(PAR_SphereSources=DataInterface() {propertyValues {EmitterBinding {bindingMode=Other, emitterName=SphereSources}}}) {module=PAR_GatherPositions_02}
   &self:PAR_GatherPositions_MinMax(PAR_SphereSources=DataInterface() {propertyValues {EmitterBinding {bindingMode=Other, emitterName=SphereSources}}}, ScaleRadius=1.3) {module=PAR_GatherPositions_MinMax}
   /SetVariables(Particles.NotScale:Vector3f=&Multiply_VectorByFloat(Vector=$Particles.Scale, Float=0.01)) {isSetParameters=true}
   /SetVariables(Particles.Position=&Clamp_Vector(
     Vector:Vector3f=Linked() {linkedVariable {name=Particles.Position, type=NiagaraPosition}}
     Min:Vector3f=<-30 -30 -30>
     Max:Vector3f=<30 30 30>
    )) {isSetParameters=true}
   /SetVariables(Particles.Scale=&self:ScratchDynamicInput()) {isSetParameters=true}
   &self:FixupStuck_01(
    BoxScale:Vector3f=$Particles.BoxScale
    MinScale=33
    Offset:Vector3f=$Particles.Offset
    Position:NiagaraPosition=$Particles.SkeletalMeshLocation.SampledPosition
   ) {module=FixupStuck_01}
  ]
  rendererClasses [&NiagaraMeshRendererProperties]
  renderers [
   {rendererClass=&NiagaraMeshRendererProperties, rendererIndex=0}
   {rendererClass=&NiagaraMeshRendererProperties, rendererIndex=1}
   {rendererClass=&NiagaraMeshRendererProperties, rendererIndex=2}
   {rendererClass=&NiagaraMeshRendererProperties, rendererIndex=3}
  ]
 }
}
summary {
 systemName=NS_QuartzBlob
 emitters [
  {emitterName=SphereSources, rendererClasses [&NiagaraMeshRendererProperties]}
  {
   bEnabled=false
   emitterName=Orbiting_Bits
   rendererClasses [&NiagaraSpriteRendererProperties, &NiagaraMeshRendererProperties]
  }
 ]
 userVariables [
  User.GameTimeFromBP {type=NiagaraFloat, defaultValue=0, description=""}
  User.Skelmesh {
   type=NiagaraDataInterfaceSkeletalMesh
   defaultValue {
    struct=NiagaraExt_VariableValue_DataInterface
    value {dataInterface=&self:NiagaraDataInterfaceSkeletalMesh_0}
   }
  }
 ]
}
